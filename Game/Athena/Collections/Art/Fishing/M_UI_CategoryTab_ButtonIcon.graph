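# Translucent UI material for a category tab button icon in the Collections screen, defaulting to a fish icon.
# The icon texture scales about its center and passes through SmoothStep to form an adjustable-edge mask, tinted between base and selected colors by Selected.
# CollectionsQuestV3 blends to an alternate style: a rounded-mask shape filled with a linear gradient, with the icon in white, and Quest3 colors replacing the base and selected ones.
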Material /Game/Athena/Collections/Art/Fishing/M_UI_CategoryTab_ButtonIcon
graph {
 properties {blendMode=BLEND_Translucent, materialDomain=MD_UI, pixelDepthOffsetMode=PDOM_Legacy, usedShadingModels=Default Lit}
 nodes [
  n1 = ScalarParameter() {defaultValue=1, parameterName=Scale}
  n2 = ScalarParameter() {defaultValue=0.2, parameterName=IconEdge}
  n3 = MaterialFunctionCall(
   Alpha=TextureSample(
    UVs=MaterialFunctionCall(Texture Scale=@n1) {materialFunction=&ScaleUVsByCenter}.UVs
    Tex=TextureObjectParameter() {group=Item, parameterName=ItemIcon, texture=&/T_Ui_collection_FIsh128px}
   ) {samplerSource=SSM_Clamp_WorldGroupSettings, texture=&T_Ui_collection_FIsh128px}.RGB
   Min=@n2
   Max=Add(A=@n2, B=ScalarParameter() {defaultValue=0.01, parameterName=IconEdgeSoftness})
  ) {materialFunction=&SmoothStep}
  LinearInterpolate(
   A=@n3.Result
   B=MaterialFunctionCall(Corner Radius=Constant() {r=0.25}, Radius in Pixels=Constant()) {materialFunction=&MF_UI_CoreUI_RoundedMask}.Shape Mask
   Alpha=ScalarParameter() {group=Quest3, parameterName=CollectionsQuestV3}
  )
  n4 = NamedRerouteUsage()
  n5 = NamedRerouteUsage()
  n6 = Power(Base=MaterialFunctionCall(RotationAngle=Constant() {r=0.15}, GradientOffset=Constant() {r=-0.2}) {materialFunction=&MF_UI_LinearGradient}.Result)
  LinearInterpolate(
   A=LinearInterpolate(A=NamedRerouteUsage(), B=NamedRerouteUsage(), Alpha=ScalarParameter() {parameterName=Selected, sliderMax=1})
   B=LinearInterpolate(A=LinearInterpolate(A=@n4, B=@n5, Alpha=@n6), Alpha=@n3.Result)
   Alpha=ScalarParameter() {group=Quest3, parameterName=CollectionsQuestV3}
  )
  NamedRerouteDeclaration(None=LinearInterpolate(
    A=VectorParameter() {parameterName=BaseRGB, defaultValue=<rgba 0 0.0103 0.3185 1>}.RGB
    B=VectorParameter() {group=Quest3, parameterName=LeftColor_Quest3, defaultValue=<rgba 0 0.0103 0.3185 1>}.RGB
    Alpha=ScalarParameter() {group=Quest3, parameterName=CollectionsQuestV3}
   )) {name=BaseRGB, nodeColor=<rgba 1 0 0.8906 1>}
  NamedRerouteDeclaration(None=LinearInterpolate(
    A=VectorParameter() {parameterName=SelectedRGB, defaultValue=<rgba 0 0.402 1 1>}.RGB
    B=VectorParameter() {group=Quest3, parameterName=RightColor_Quest3, defaultValue=<rgba 0 0.402 1 1>}.RGB
    Alpha=ScalarParameter() {group=Quest3, parameterName=CollectionsQuestV3}
   )) {name=SelectedRGB, nodeColor=<rgba 0 1 0.7891 1>}
 ]
}
