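# Translucent UI material that draws a fish in the collection detail view, with a static switch choosing the fish icon or a smoothstepped placeholder glyph.
# The icon is scaled, mirrored, bobbed by a sine and wiggled by a custom-HLSL wave masked toward the tail.
# IsDiscovered blends from dark silhouette to icon color, under a vertical gradient tint.
# Two panning, hash-randomized bubble layers and a radial bloom add to opacity.
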Material /Game/Athena/Collections/Art/Fishing/M_Ui_CollectionDetail_FIsh
graph {
 properties {blendMode=BLEND_Translucent, materialDomain=MD_UI, pixelDepthOffsetMode=PDOM_Legacy, usedShadingModels=Default Lit}
 nodes [
  n1 = ScalarParameter() {defaultValue=0.5, parameterName=IconScale}
  n2 = ScalarParameter() {defaultValue=0.2, parameterName=IconEdge}
  n3 = MaterialFunctionCall(
   Alpha=TextureSample(UVs=MaterialFunctionCall(Texture Scale=@n1) {materialFunction=&ScaleUVsByCenter}.UVs) {samplerSource=SSM_Clamp_WorldGroupSettings, texture=&/T_Ui_collection_FIsh128px}.RGB
   Min=@n2
   Max=Add(A=@n2, B=ScalarParameter() {defaultValue=0.01, parameterName=IconEdgeSoftness})
  ) {materialFunction=&SmoothStep}
  n4 = Saturate(None=Subtract(A=@n3.Result))
  n5 = Constant() {r=1}
  n6 = LinearInterpolate(
   A=ScalarParameter() {defaultValue=0.2, group=Item, parameterName=ItemSize_Min, sliderMax=1}
   B=ScalarParameter() {defaultValue=0.5, group=Item, parameterName=ItemSize_Max, sliderMax=1}
   Alpha=ScalarParameter() {defaultValue=0.5, group=Item, parameterName=RelativeSize, sliderMax=1}
  )
  n7 = LinearInterpolate(B=AppendVector(A=Multiply(A=@n6) {constB=-1}, B=@n6), Alpha=ScalarParameter() {defaultValue=1, parameterName=Intro, sliderMax=1})
  n8 = Constant()
  n9 = ScalarParameter() {defaultValue=0.05, group=Hovering, parameterName=HoveringOffset_Down, sliderMax=1}
  n10 = Add(A=Time(), B=ScalarParameter() {group=Hovering, parameterName=RandomizeTime})
  n11 = ScalarParameter() {defaultValue=0.25, group=Hovering, parameterName=hoveringSpeed, sliderMax=1}
  n12 = LinearInterpolate(A=Multiply(A=@n9) {constB=-1}, B=@n9, Alpha=Sine(None=Multiply(A=@n10, B=@n11) {constB=0.25})) {constA=-0.05, constB=0.05}
  MaterialExpressionReroute_23 = Reroute(None=Add(
    A=MaterialFunctionCall(Texture Scale=@n7) {materialFunction=&ScaleUVsByCenter}.UVs
    B=AppendVector(A=@n8, B=@n12)
   ))
  MaterialExpressionTime_3 = Time()
  MaterialExpressionScalarParameter_91 = ScalarParameter() {defaultValue=0.036, group=tail, parameterName=OffsetRange}
  MaterialExpressionScalarParameter_97 = ScalarParameter() {defaultValue=0.9, group=tail, parameterName=TimeMult}
  MaterialExpressionScalarParameter_98 = ScalarParameter() {defaultValue=1.5, group=tail, parameterName=ScaleA}
  MaterialExpressionScalarParameter_99 = ScalarParameter() {group=tail, parameterName=AddA}
  MaterialExpressionScalarParameter_100 = ScalarParameter() {defaultValue=0.3, group=tail, parameterName=TurbA}
  MaterialExpressionScalarParameter_101 = ScalarParameter() {defaultValue=0.6, group=tail, parameterName=TurbB}
  n13 = Custom!(
   inUV=@MaterialExpressionReroute_23
   inTime=@MaterialExpressionTime_3
   inValRange=@MaterialExpressionScalarParameter_91
   inTimeScale=@MaterialExpressionScalarParameter_97
   sinScaleA=@MaterialExpressionScalarParameter_98
   sinAddA=@MaterialExpressionScalarParameter_99
   sinTurbA=@MaterialExpressionScalarParameter_100
   sinTurbB=@MaterialExpressionScalarParameter_101
  ) {
   code="   float wave = sin((inUV.y + inUV.x - inTime * inTimeScale + sin(sinScaleA * inUV.y + sinAddA * inUV.x) * PI * sinTurbA)\r\n                  * PI * sinTurbB);\r\n    \r\n   return wave;"
   description=Wave
   outputType=CMOT_Float1
   inputs [
    {inputName=inUV, input {expression=&self:MaterialExpressionReroute_23}}
    {inputName=inTime, input {expression=&self:MaterialExpressionTime_3}}
    {
     inputName=inValRange
     input {expression=&self:MaterialExpressionScalarParameter_91}
    }
    {
     inputName=inTimeScale
     input {expression=&self:MaterialExpressionScalarParameter_97}
    }
    {inputName=sinScaleA, input {expression=&self:MaterialExpressionScalarParameter_98}}
    {inputName=sinAddA, input {expression=&self:MaterialExpressionScalarParameter_99}}
    {inputName=sinTurbA, input {expression=&self:MaterialExpressionScalarParameter_100}}
    {inputName=sinTurbB, input {expression=&self:MaterialExpressionScalarParameter_101}}
   ]
  }
  n14 = Subtract(A=@MaterialExpressionScalarParameter_91, B=Multiply(A=@MaterialExpressionScalarParameter_91, B=@n13))
  n15 = ComponentMask(None=Rotator(Coordinate=@MaterialExpressionReroute_23, Time=ScalarParameter() {group=tail, parameterName=TailMaskRotation, sliderMax=1}) {speed=6.2832}) {r=true}
  n16 = ScalarParameter() {defaultValue=0.25, group=tail, parameterName=TailMaskOffset, sliderMax=1}
  n17 = LinearInterpolate(
   A=@MaterialExpressionReroute_23
   B=Multiply(A=Add(A=@n5, B=@n14), B=@MaterialExpressionReroute_23)
   Alpha=Saturate(None=Subtract(A=@n15, B=@n16))
  )
  n18 = TextureSample(
   UVs=@n17
   Tex=TextureObjectParameter() {group=Item, parameterName=ItemIcon, texture=/Game/UI/Foundation/Textures/Icons/Athena/T-T-Icon-BR-FlopperGadget-L}
  ) {samplerSource=SSM_Clamp_WorldGroupSettings, texture=/Game/UI/Foundation/Textures/Icons/Athena/T-T-Icon-BR-FlopperGadget}
  n19 = MaterialFunctionCall!(UVs=@n17, Alpha=@n18.A) {materialFunction=/Game/UI/Foundation/Items/ClampAlphaToTransparentBorder}
  n20 = MaterialFunctionCall(Float2=MaterialFunctionCall() {materialFunction=&ScreenResolution}.Visible Resolution) {materialFunction=&BreakOutFloat2Components}
  n21 = Max(A=@n20.G)
  n22 = Multiply(
   A=MaterialFunctionCall() {materialFunction=&ScreenAlignedUVs}."X 100%, Y 100%"
   B=AppendVector(A=Divide(A=@n20.R, B=@n21), B=Constant() {r=1})
  )
  n23 = ScalarParameter() {defaultValue=4, group=effect, parameterName=TileScreenSpace}
  n24 = MaterialFunctionCall() {materialFunction=&GetUserInterfaceUV}
  n25 = MaterialFunctionCall(Float2=@n24.Pixel Size) {materialFunction=&BreakOutFloat2Components}
  n26 = Max(A=@n25.G)
  n27 = Multiply(A=AppendVector(A=Divide(A=@n25.R, B=@n26), B=Constant() {r=1}), B=@n24.Normalized UV)
  n28 = ScalarParameter() {defaultValue=0.25, group=effect, parameterName=Local UV}
  n29 = Multiply(
   A=Add(A=Multiply(A=@n22, B=@n23), B=Multiply(A=@n27, B=@n28))
   B=ScalarParameter() {defaultValue=0.5, group=effect, parameterName=TileAmount}
  )
  n30 = Multiply(A=Panner(Coordinate=@n29) {speedY=0.1}) {constB=4}
  n31 = Frac(None=@n30)
  n32 = MaterialFunctionCall(Float2=MaterialFunctionCall(Vec2=Floor(None=@n30)) {materialFunction=&MF_Hash}.Vec2) {materialFunction=&BreakOutFloat2Components}
  n33 = Multiply(A=AppendVector(A=@n32.R, B=@n32.R), B=ScalarParameter() {defaultValue=0.5, group=effect, parameterName=Bubble_translation}) {constB=0.2}
  n34 = Subtract(A=@n32.R, B=ScalarParameter() {defaultValue=0.25, group=effect, parameterName=BubbleAmount}) {constB=0.25}
  n35 = TextureSample(UVs=MaterialFunctionCall(UVs=Add(A=@n31, B=@n33), Texture Scale=Saturate(None=@n34)) {materialFunction=&ScaleUVsByCenter}.UVs) {samplerSource=SSM_Clamp_WorldGroupSettings, texture=/Game/Athena/Collections/Art/T_Ui_Bubble_64}
  n36 = Multiply(A=Panner(Coordinate=@n29) {speedY=0.05}) {constB=3}
  n37 = MaterialFunctionCall(Float2=MaterialFunctionCall(Vec2=Floor(None=@n36)) {materialFunction=&MF_Hash}.Vec2) {materialFunction=&BreakOutFloat2Components}
  n38 = AppendVector(A=@n37.R, B=@n37.R)
  n39 = ScalarParameter() {defaultValue=0.5, group=effect, parameterName=Bubble_translation}
  n40 = ScalarParameter() {defaultValue=0.25, group=effect, parameterName=BubbleAmount}
  n41 = MaterialFunctionCall(UVs=Add(A=Frac(None=@n36), B=Multiply(A=@n38, B=@n39) {constB=0.2}), Texture Scale=Saturate(None=Subtract(A=@n37.R, B=@n40) {constB=0.25})) {materialFunction=&ScaleUVsByCenter}
  n42 = Multiply(
   A=Add(
    A=Multiply(A=@n35.G) {constB=0.5}
    B=TextureSample(UVs=@n41.UVs) {samplerSource=SSM_Clamp_WorldGroupSettings, texture=&T_Ui_Bubble_64}.R
   )
   B=Saturate(None=MaterialFunctionCall() {materialFunction=&RadialGradientExponential}.RadialGradientExponential)
  )
  n43 = Saturate(None=MaterialFunctionCall(
    Radius=ScalarParameter() {defaultValue=0.5, group=bloom, parameterName=Bloom_Radius}
    Density=ScalarParameter() {defaultValue=0.5, group=bloom, parameterName=Bloom_Density}
   ) {materialFunction=&RadialGradientExponential}.RadialGradientExponential)
  Saturate(None=Add(A=StaticSwitchParameter(True=@n19.Result, False=@n4) {defaultValue=true, parameterName=is known}, B=Add(A=@n42, B=@n43)))
  n44 = VectorParameter() {group=background, parameterName=Background_glow_T, defaultValue=<rgba 0 0.6654 1 1>}
  n45 = LinearInterpolate(
   A=VectorParameter() {parameterName=Background_topgradient_F, defaultValue=<rgba 0 0.004 0.0723 1>}.RGB
   B=@n18.RGB
   Alpha=ScalarParameter() {group=control, parameterName=IsDiscovered, sliderMax=1}
  )
  n46 = ComponentMask(None=TextureCoordinate()) {g=true}
  n47 = ScalarParameter() {defaultValue=0.4, group=Item, parameterName=FallOff_Placement, sliderMax=1}
  n48 = Saturate(None=Multiply(A=Subtract(A=@n46, B=@n47) {constB=0.5}, B=ScalarParameter() {defaultValue=1, group=Item, parameterName=FallOff_Strenght}) {constB=2})
  n49 = VectorParameter() {group=background, parameterName=Background_glow_T, defaultValue=<rgba 0 0.6654 1 1>}
  n50 = VectorParameter() {group=background, parameterName=Background_topgradient_T, defaultValue=<rgba 0 0.0578 0.6038 1>}
  StaticSwitchParameter(
   True=LinearInterpolate(
    A=LinearInterpolate(A=@n44.RGB, B=@n45, Alpha=@n19.Result)
    B=VectorParameter() {group=background, parameterName=Background_topgradient_T, defaultValue=<rgba 0 0.0578 0.6038 1>}.RGB
    Alpha=Multiply(A=@n19.Result, B=@n48)
   )
   False=LinearInterpolate(A=LinearInterpolate(A=@n49.RGB, B=@n50.RGB, Alpha=@n3.Result), B=@n49.RGB, Alpha=@n42)
  ) {defaultValue=true, parameterName=is known}
 ]
}
