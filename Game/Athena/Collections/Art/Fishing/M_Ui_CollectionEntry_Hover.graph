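# Translucent UI material that draws the hover border for a collection entry in the Fishing collection art.
# Each edge is a slightly rotated UV axis thresholded by a range offset and sharpness, giving crooked, crisp lines.
# The four edge masks multiply into an inner area, which is inverted to leave the border.
# A scalar "hover" parameter blends each edge's angle and offset, and raises border opacity from faint to full by default.
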Material /Game/Athena/Collections/Art/Fishing/M_Ui_CollectionEntry_Hover
graph {
 properties {blendMode=BLEND_Translucent, materialDomain=MD_UI, pixelDepthOffsetMode=PDOM_Legacy, usedShadingModels=Default Lit}
 nodes [
  n1 = ScalarParameter() {defaultValue=0.002, group=Border Angles, parameterName=Top-Angle}
  n2 = ScalarParameter() {defaultValue=-0.002, group=Border Angles, parameterName=Top-Angle-hover}
  n3 = ScalarParameter() {group=control, parameterName=hover, sliderMax=1}
  n4 = ComponentMask(None=Rotator(
    Coordinate=MaterialFunctionCall() {materialFunction=&GetUserInterfaceUV}.9-Slice UV
    Time=LinearInterpolate(A=@n1, B=@n2, Alpha=@n3)
   ) {speed=6.2832}) {g=true}
  n5 = LinearInterpolate(
   A=ScalarParameter() {defaultValue=0.035, group=Border_Range, parameterName=Top-RangeOffset, sliderMax=1}
   B=ScalarParameter() {defaultValue=0.035, group=Border_Range, parameterName=Top-RangeOffset-hover, sliderMax=1}
   Alpha=ScalarParameter() {group=control, parameterName=hover, sliderMax=1}
  )
  n6 = Saturate(None=Multiply(A=Subtract(A=@n4, B=@n5), B=ScalarParameter() {defaultValue=512, group=Border sharpness, parameterName=HorizontalSharpness}))
  n7 = ScalarParameter() {defaultValue=-0.002, group=Border Angles, parameterName=Left-Angle}
  n8 = ScalarParameter() {defaultValue=0.001, group=Border Angles, parameterName=Left-Angle-hover}
  n9 = ScalarParameter() {group=control, parameterName=hover, sliderMax=1}
  n10 = ComponentMask(None=Rotator(Coordinate=TextureCoordinate() {coordinateIndex=4}, Time=LinearInterpolate(A=@n7, B=@n8, Alpha=@n9)) {speed=6.2832}) {r=true}
  n11 = LinearInterpolate(
   A=ScalarParameter() {defaultValue=0.035, group=Border_Range, parameterName=Left-RangeOffset, sliderMax=1}
   B=ScalarParameter() {defaultValue=0.035, group=Border_Range, parameterName=Left-RangeOffset-hover, sliderMax=1}
   Alpha=ScalarParameter() {group=control, parameterName=hover, sliderMax=1}
  )
  n12 = Saturate(None=Multiply(A=Subtract(A=@n10, B=@n11), B=ScalarParameter() {defaultValue=512, group=Border sharpness, parameterName=VerticalSharpness}))
  n13 = ScalarParameter() {defaultValue=0.5025, group=Border Angles, parameterName=Bot-Angle}
  n14 = ScalarParameter() {defaultValue=0.4975, group=Border Angles, parameterName=Bot-Angle-hover}
  n15 = ScalarParameter() {group=control, parameterName=hover, sliderMax=1}
  n16 = ComponentMask(None=Rotator(Coordinate=TextureCoordinate() {coordinateIndex=4}, Time=LinearInterpolate(A=@n13, B=@n14, Alpha=@n15)) {speed=6.2832}) {g=true}
  n17 = LinearInterpolate(
   A=ScalarParameter() {defaultValue=0.035, group=Border_Range, parameterName=Bot-RangeOffset, sliderMax=1}
   B=ScalarParameter() {defaultValue=0.035, group=Border_Range, parameterName=Bot-RangeOffset-hover, sliderMax=1}
   Alpha=ScalarParameter() {group=control, parameterName=hover, sliderMax=1}
  )
  n18 = Saturate(None=Multiply(A=Subtract(A=@n16, B=@n17), B=ScalarParameter() {group=Border sharpness, parameterName=HorizontalSharpness}))
  n19 = ScalarParameter() {defaultValue=0.4975, group=Border Angles, parameterName=Right-Angle}
  n20 = ScalarParameter() {defaultValue=0.5015, group=Border Angles, parameterName=Right-Angle-hover}
  n21 = ScalarParameter() {group=control, parameterName=hover, sliderMax=1}
  n22 = ComponentMask(None=Rotator(Coordinate=TextureCoordinate() {coordinateIndex=4}, Time=LinearInterpolate(A=@n19, B=@n20, Alpha=@n21)) {speed=6.2832}) {r=true}
  n23 = LinearInterpolate(
   A=ScalarParameter() {defaultValue=0.035, group=Border_Range, parameterName=Right-RangeOffset, sliderMax=1}
   B=ScalarParameter() {defaultValue=0.035, group=Border_Range, parameterName=Right-RangeOffset-hover, sliderMax=1}
   Alpha=ScalarParameter() {group=control, parameterName=hover, sliderMax=1}
  )
  n24 = Saturate(None=Multiply(A=Subtract(A=@n22, B=@n23), B=ScalarParameter() {group=Border sharpness, parameterName=VerticalSharpness}))
  n25 = Saturate(None=DotProduct(A=DotProduct(A=@n6, B=@n12), B=DotProduct(A=@n18, B=@n24)))
  n26 = ScalarParameter() {defaultValue=0.05, group=Border Alpha, parameterName=BorderAlpha, sliderMax=1}
  n27 = ScalarParameter() {defaultValue=1, group=Border Alpha, parameterName=HoveredBorderAlpha, sliderMax=1, sliderMin=0.01}
  n28 = ScalarParameter() {group=control, parameterName=hover, sliderMax=1}
  Saturate(None=Multiply(A=OneMinus(None=@n25), B=LinearInterpolate(A=@n26, B=@n27, Alpha=@n28) {constB=0}))
  TextureCoordinate() {coordinateIndex=4}
  VectorParameter() {parameterName=RGB, defaultValue=<rgba 1 1 1 1>}
 ]
}
