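# Translucent UI material for a fishing collection entry, showing a fish icon with score stars, or a faint placeholder icon via the "is known" static switch.
# The item icon bobs on a sine, with a Custom-node wave distorting its tail-side UVs.
# Four rotated, thresholded UV edges form a skewed border that shifts on hover and changes opacity when selected.
# IsDiscovered, Has2Stars and Has3Stars parameters drive icon color and star display.
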Material /Game/Athena/Collections/Art/Fishing/M_Ui_CollectionNewEntry
graph {
 properties {blendMode=BLEND_Translucent, materialDomain=MD_UI, pixelDepthOffsetMode=PDOM_Legacy, usedShadingModels=Default Lit}
 nodes [
  n1 = VectorParameter() {group=background, parameterName=Background_Stars, defaultValue=<rgba 0 0.6654 1 0.5>}
  n2 = MaterialFunctionCall(Float2=MaterialFunctionCall() {materialFunction=&GetUserInterfaceUV}.Pixel Size) {materialFunction=&BreakOutFloat2Components}
  n3 = Max(A=@n2.G)
  n4 = MaterialFunctionCall(Float2=TextureCoordinate() {coordinateIndex=4}) {materialFunction=&BreakOutFloat2Components}
  n5 = AppendVector(A=Multiply(A=Divide(A=@n2.R, B=@n3), B=@n4.R), B=@n4.G)
  n6 = LinearInterpolate(
   A=AppendVector(
    A=ScalarParameter() {defaultValue=0.4, group=ScoreIcon, parameterName=Scoreposition_U}
    B=ScalarParameter() {defaultValue=0.38, group=ScoreIcon, parameterName=Scoreposition_V}
   )
   B=AppendVector(
    A=ScalarParameter() {defaultValue=0.4, group=ScoreIcon, parameterName=Scoreposition_U_Hover}
    B=ScalarParameter() {defaultValue=0.38, group=ScoreIcon, parameterName=Scoreposition_V_Hover}
   )
   Alpha=ScalarParameter() {group=control, parameterName=Selected, sliderMax=1}
  )
  n7 = TextureSample(UVs=MaterialFunctionCall(UVs=Add(A=@n5, B=@n6), Texture Scale=ScalarParameter() {defaultValue=0.2, group=ScoreIcon, parameterName=ScoreIconScale, sliderMax=1}) {materialFunction=&ScaleUVsByCenter}.UVs) {samplerSource=SSM_Clamp_WorldGroupSettings, texture=/Game/Athena/Collections/Art/T_Ui_Collection_Score_64}
  n8 = ComponentMask(None=@n7.R) {r=true}
  n9 = LinearInterpolate(
   A=LinearInterpolate(A=@n7.B, B=@n7.B, Alpha=ScalarParameter() {defaultValue=1, group=control, parameterName=Has2Stars, sliderMax=1})
   B=@n7.R
   Alpha=ScalarParameter() {group=control, parameterName=Has3Stars, sliderMax=1}
  )
  n10 = Add(A=Multiply(A=@n1.A, B=@n8), B=@n9)
  n11 = Rotator(
   Coordinate=TextureCoordinate() {coordinateIndex=4}
   Time=LinearInterpolate(
    A=ScalarParameter() {defaultValue=0.002, group=Border Angles, parameterName=Top-Angle}
    B=ScalarParameter() {defaultValue=-0.002, group=Border Angles, parameterName=Top-Angle-hover}
    Alpha=ScalarParameter() {group=control, parameterName=hover, sliderMax=1}
   )
  ) {speed=6.2832}
  n12 = ScalarParameter() {defaultValue=0.035, group=Border_Range, parameterName=Top-RangeOffset, sliderMax=1}
  n13 = ScalarParameter() {defaultValue=0.035, group=Border_Range, parameterName=Top-RangeOffset-hover, sliderMax=1}
  n14 = ScalarParameter() {group=control, parameterName=hover, sliderMax=1}
  n15 = Multiply(
   A=Subtract(A=ComponentMask(None=@n11) {g=true}, B=LinearInterpolate(A=@n12, B=@n13, Alpha=@n14))
   B=ScalarParameter() {defaultValue=512, group=Border sharpness, parameterName=HorizontalSharpness}
  )
  n16 = Rotator(
   Coordinate=TextureCoordinate() {coordinateIndex=4}
   Time=LinearInterpolate(
    A=ScalarParameter() {defaultValue=-0.002, group=Border Angles, parameterName=Left-Angle}
    B=ScalarParameter() {defaultValue=0.001, group=Border Angles, parameterName=Left-Angle-hover}
    Alpha=ScalarParameter() {group=control, parameterName=hover, sliderMax=1}
   )
  ) {speed=6.2832}
  n17 = ScalarParameter() {defaultValue=0.035, group=Border_Range, parameterName=Left-RangeOffset, sliderMax=1}
  n18 = ScalarParameter() {defaultValue=0.035, group=Border_Range, parameterName=Left-RangeOffset-hover, sliderMax=1}
  n19 = ScalarParameter() {group=control, parameterName=hover, sliderMax=1}
  n20 = Multiply(
   A=Subtract(A=ComponentMask(None=@n16) {r=true}, B=LinearInterpolate(A=@n17, B=@n18, Alpha=@n19))
   B=ScalarParameter() {defaultValue=512, group=Border sharpness, parameterName=VerticalSharpness}
  )
  n21 = Rotator(
   Coordinate=TextureCoordinate() {coordinateIndex=4}
   Time=LinearInterpolate(
    A=ScalarParameter() {defaultValue=0.5025, group=Border Angles, parameterName=Bot-Angle}
    B=ScalarParameter() {defaultValue=0.4975, group=Border Angles, parameterName=Bot-Angle-hover}
    Alpha=ScalarParameter() {group=control, parameterName=hover, sliderMax=1}
   )
  ) {speed=6.2832}
  n22 = ScalarParameter() {defaultValue=0.035, group=Border_Range, parameterName=Bot-RangeOffset, sliderMax=1}
  n23 = ScalarParameter() {defaultValue=0.035, group=Border_Range, parameterName=Bot-RangeOffset-hover, sliderMax=1}
  n24 = ScalarParameter() {group=control, parameterName=hover, sliderMax=1}
  n25 = Multiply(
   A=Subtract(A=ComponentMask(None=@n21) {g=true}, B=LinearInterpolate(A=@n22, B=@n23, Alpha=@n24))
   B=ScalarParameter() {defaultValue=512, group=Border sharpness, parameterName=HorizontalSharpness}
  )
  n26 = Rotator(
   Coordinate=TextureCoordinate() {coordinateIndex=4}
   Time=LinearInterpolate(
    A=ScalarParameter() {defaultValue=0.4975, group=Border Angles, parameterName=Right-Angle}
    B=ScalarParameter() {defaultValue=0.5015, group=Border Angles, parameterName=Right-Angle-hover}
    Alpha=ScalarParameter() {group=control, parameterName=hover, sliderMax=1}
   )
  ) {speed=6.2832}
  n27 = ScalarParameter() {defaultValue=0.035, group=Border_Range, parameterName=Right-RangeOffset, sliderMax=1}
  n28 = ScalarParameter() {defaultValue=0.035, group=Border_Range, parameterName=Right-RangeOffset-hover, sliderMax=1}
  n29 = ScalarParameter() {group=control, parameterName=hover, sliderMax=1}
  n30 = Multiply(
   A=Subtract(A=ComponentMask(None=@n26) {r=true}, B=LinearInterpolate(A=@n27, B=@n28, Alpha=@n29))
   B=ScalarParameter() {defaultValue=512, group=Border sharpness, parameterName=VerticalSharpness}
  )
  n31 = DotProduct(A=DotProduct(A=Saturate(None=@n15), B=Saturate(None=@n20)), B=DotProduct(A=Saturate(None=@n25), B=Saturate(None=@n30)))
  n32 = Saturate(None=OneMinus(None=Saturate(None=@n31)))
  n33 = ScalarParameter() {defaultValue=0.25, group=Border Alpha, parameterName=BorderAlpha, sliderMax=1}
  n34 = ScalarParameter() {defaultValue=1, group=Border Alpha, parameterName=HoveredBorderAlpha, sliderMax=1, sliderMin=0.01}
  n35 = ScalarParameter() {group=control, parameterName=Selected, sliderMax=1}
  n36 = Saturate(None=Multiply(A=@n32, B=LinearInterpolate(A=@n33, B=@n34, Alpha=@n35) {constB=0}))
  n37 = ScalarParameter() {defaultValue=0.25, group=Border Alpha, parameterName=EmptyIconAlpha, sliderMax=1}
  n38 = ScalarParameter() {defaultValue=0.25, group=Border Alpha, parameterName=EmptyIconAlphaHover, sliderMax=1, sliderMin=0.01}
  n39 = ScalarParameter() {group=control, parameterName=hover, sliderMax=1}
  n40 = MaterialFunctionCall(Float2=MaterialFunctionCall() {materialFunction=&GetUserInterfaceUV}.Pixel Size) {materialFunction=&BreakOutFloat2Components}
  n41 = Max(A=@n40.G)
  n42 = MaterialFunctionCall(Float2=TextureCoordinate() {coordinateIndex=4}) {materialFunction=&BreakOutFloat2Components}
  n43 = AppendVector(A=Multiply(A=Divide(A=@n40.R, B=@n41), B=@n42.R), B=@n42.G)
  n44 = AppendVector(
   A=ScalarParameter() {defaultValue=0.09, group=Item, parameterName=Itemposition_U}
   B=ScalarParameter() {defaultValue=0.04, group=Item, parameterName=Itemposition_V}
  )
  n45 = ScalarParameter() {defaultValue=0.5, parameterName=EmptyIconScale}
  n46 = ScalarParameter() {defaultValue=0.6, parameterName=EmptyIconScale_Hover}
  n47 = ScalarParameter() {group=control, parameterName=hover, sliderMax=1, sliderMin=0.1}
  n48 = TextureSampleParameter2D(UVs=MaterialFunctionCall(UVs=Add(A=@n43, B=@n44), Texture Scale=LinearInterpolate(A=@n45, B=@n46, Alpha=@n47) {constB=0.01}) {materialFunction=&ScaleUVsByCenter}.UVs) {
   parameterName=EmptyStateIcon
   samplerSource=SSM_Clamp_WorldGroupSettings
   texture=&/T_Ui_collection_FIsh128px
  }
  n49 = ScalarParameter() {defaultValue=0.2, parameterName=IconEdge}
  n50 = Add(A=@n49, B=ScalarParameter() {defaultValue=0.01, parameterName=IconEdgeSoftness})
  n51 = Constant() {r=1}
  n52 = MaterialFunctionCall(Float2=MaterialFunctionCall() {materialFunction=&GetUserInterfaceUV}.Pixel Size) {materialFunction=&BreakOutFloat2Components}
  n53 = Divide(A=@n52.R, B=Max(A=@n52.G))
  n54 = MaterialFunctionCall(Float2=TextureCoordinate() {coordinateIndex=4}) {materialFunction=&BreakOutFloat2Components}
  n55 = Add(
   A=AppendVector(A=Multiply(A=@n53, B=@n54.R), B=@n54.G)
   B=AppendVector(
    A=ScalarParameter() {defaultValue=0.09, group=Item, parameterName=Itemposition_U}
    B=ScalarParameter() {defaultValue=0.04, group=Item, parameterName=Itemposition_V}
   )
  )
  n56 = Add(
   A=LinearInterpolate(
    A=ScalarParameter() {defaultValue=0.2, group=Item, parameterName=ItemSize_Min, sliderMax=1}
    B=ScalarParameter() {defaultValue=0.8, group=Item, parameterName=ItemSize_Max, sliderMax=1}
    Alpha=ScalarParameter() {defaultValue=0.5, group=Item, parameterName=ItemSize, sliderMax=1}
   )
   B=LinearInterpolate(
    B=ScalarParameter() {defaultValue=0.01, group=Item, parameterName=HoveredItemsize}
    Alpha=ScalarParameter() {group=control, parameterName=hover, sliderMax=1, sliderMin=0.1}
   ) {constB=0.01}
  )
  n57 = AppendVector(A=Multiply(A=@n56) {constB=-1}, B=@n56)
  n58 = Constant()
  n59 = ScalarParameter() {defaultValue=0.03, group=Hovering, parameterName=HoveringOffset_Down, sliderMax=1}
  n60 = Add(A=Time(), B=ScalarParameter() {defaultValue=0.5, group=Hovering, parameterName=RandomizeTime})
  n61 = LinearInterpolate(
   A=ScalarParameter() {defaultValue=0.05, group=Hovering, parameterName=hoveringSpeed_Min, sliderMax=1}
   B=ScalarParameter() {defaultValue=0.2, group=Hovering, parameterName=hoveringSpeed_Max, sliderMax=1}
   Alpha=ScalarParameter() {defaultValue=0.5, group=Hovering, parameterName=RandomizeSpeed}
  )
  n62 = LinearInterpolate(A=Multiply(A=@n59) {constB=-1}, B=@n59, Alpha=Sine(None=Multiply(A=@n60, B=@n61) {constB=0.25})) {constA=-0.05, constB=0.05}
  MaterialExpressionReroute_23 = Reroute(None=Add(
    A=MaterialFunctionCall(UVs=@n55, Texture Scale=@n57) {materialFunction=&ScaleUVsByCenter}.UVs
    B=AppendVector(A=@n58, B=@n62)
   ))
  MaterialExpressionTime_3 = Time()
  MaterialExpressionScalarParameter_91 = ScalarParameter() {defaultValue=0.25, group=tail, parameterName=OffsetRange}
  MaterialExpressionScalarParameter_97 = ScalarParameter() {defaultValue=0.9, group=tail, parameterName=TimeMult}
  MaterialExpressionScalarParameter_98 = ScalarParameter() {defaultValue=1.5, group=tail, parameterName=ScaleA}
  MaterialExpressionScalarParameter_99 = ScalarParameter() {group=tail, parameterName=AddA}
  MaterialExpressionScalarParameter_100 = ScalarParameter() {defaultValue=0.3, group=tail, parameterName=TurbA}
  MaterialExpressionScalarParameter_101 = ScalarParameter() {defaultValue=0.6, group=tail, parameterName=TurbB}
  n63 = Custom!(
   inUV=@MaterialExpressionReroute_23
   inTime=@MaterialExpressionTime_3
   inValRange=@MaterialExpressionScalarParameter_91
   inTimeScale=@MaterialExpressionScalarParameter_97
   sinScaleA=@MaterialExpressionScalarParameter_98
   sinAddA=@MaterialExpressionScalarParameter_99
   sinTurbA=@MaterialExpressionScalarParameter_100
   sinTurbB=@MaterialExpressionScalarParameter_101
  ) {
   code="   float wave = sin((inUV.y + inUV.x - inTime * inTimeScale + sin(sinScaleA * inUV.y + sinAddA * inUV.x) * PI * sinTurbA)\r\n                  * PI * sinTurbB);\r\n    \r\n   return wave;"
   description=Wave
   outputType=CMOT_Float1
   inputs [
    {inputName=inUV, input {expression=&self:MaterialExpressionReroute_23}}
    {inputName=inTime, input {expression=&self:MaterialExpressionTime_3}}
    {inputName=inValRange, input {expression=&self:MaterialExpressionScalarParameter_91}}
    {inputName=inTimeScale, input {expression=&self:MaterialExpressionScalarParameter_97}}
    {inputName=sinScaleA, input {expression=&self:MaterialExpressionScalarParameter_98}}
    {inputName=sinAddA, input {expression=&self:MaterialExpressionScalarParameter_99}}
    {inputName=sinTurbA, input {expression=&self:MaterialExpressionScalarParameter_100}}
    {inputName=sinTurbB, input {expression=&self:MaterialExpressionScalarParameter_101}}
   ]
  }
  n64 = Subtract(A=@MaterialExpressionScalarParameter_91, B=Multiply(A=@MaterialExpressionScalarParameter_91, B=@n63))
  n65 = ComponentMask(None=@MaterialExpressionReroute_23) {r=true}
  n66 = ScalarParameter() {defaultValue=0.6, group=tail, parameterName=TailMask, sliderMax=1}
  n67 = LinearInterpolate(
   A=@MaterialExpressionReroute_23
   B=Multiply(A=Add(A=@n51, B=@n64), B=@MaterialExpressionReroute_23)
   Alpha=Saturate(None=Subtract(A=@n65, B=@n66))
  )
  n68 = TextureObjectParameter() {
   group=Item
   parameterName=ItemIcon
   samplerSource=SSM_Clamp_WorldGroupSettings
   texture=/Game/UI/Foundation/Textures/Icons/Athena/T-T-Icon-BR-FlopperGadget
  }
  n69 = TextureSample(UVs=@n67, Tex=@n68) {samplerSource=SSM_Clamp_WorldGroupSettings, texture=&T-T-Icon-BR-FlopperGadget}
  n70 = MaterialFunctionCall!(UVs=@n67, Alpha=@n69.A) {materialFunction=/Game/UI/Foundation/Items/ClampAlphaToTransparentBorder}
  n71 = StaticSwitchParameter(
   True=@n70.Result
   False=Multiply(
    A=LinearInterpolate(A=@n37, B=@n38, Alpha=@n39) {constB=0}
    B=MaterialFunctionCall(Alpha=@n48.RGB, Min=@n49, Max=@n50) {materialFunction=&SmoothStep}.Result
   )
  ) {defaultValue=true, parameterName=is known}
  Add(
   A=LinearInterpolate(B=Saturate(None=@n10), Alpha=ScalarParameter() {defaultValue=1, group=control, parameterName=IsDiscovered, sliderMax=1})
   B=Saturate(None=Add(A=@n36, B=@n71))
  )
  n72 = VectorParameter() {parameterName=Background_topgradient_F, defaultValue=<rgba 0 0.004 0.0723 1>}
  n73 = ScalarParameter() {defaultValue=1, group=control, parameterName=IsDiscovered, sliderMax=1}
  n74 = ComponentMask(None=TextureCoordinate()) {g=true}
  n75 = ScalarParameter() {defaultValue=0.4, group=Item, parameterName=FallOff_Placement, sliderMax=1}
  n76 = ScalarParameter() {defaultValue=1, group=Item, parameterName=FallOff_Strenght}
  n77 = ScalarParameter() {defaultValue=1, group=Item, parameterName=FallOff_Strenght_Hovered}
  n78 = ScalarParameter() {defaultValue=1, group=control, parameterName=IsDiscovered, sliderMax=1}
  n79 = Saturate(None=Multiply(A=Subtract(A=@n74, B=@n75) {constB=0.5}, B=LinearInterpolate(A=@n76, B=@n77, Alpha=@n78) {constB=0}) {constB=2})
  n80 = Multiply(B=@n8) {constA=2}
  n81 = OneMinus(None=@n70.Result)
  n82 = LinearInterpolate(
   A=LinearInterpolate(
    A=LinearInterpolate(A=@n72.RGB, B=@n69.RGB, Alpha=@n73)
    B=VectorParameter() {group=background, parameterName=Background_topgradient_T, defaultValue=<rgba 0 0.0578 0.6038 1>}.RGB
    Alpha=Multiply(A=@n70.Result, B=@n79)
   )
   B=@n1.RGB
   Alpha=Multiply(A=Saturate(None=@n80), B=@n81)
  )
  n83 = LinearInterpolate(
   A=VectorParameter() {group=ScoreIcon, parameterName=Score_RGB1, defaultValue=<rgba 1 1 1 1>}.RGB
   B=VectorParameter() {group=ScoreIcon, parameterName=Score_RGB2, defaultValue=<rgba 1 0.7011 0.0144 1>}.RGB
   Alpha=ScalarParameter() {group=control, parameterName=Has3Stars, sliderMax=1}
  )
  n84 = Multiply(A=@n81, B=@n9)
  StaticSwitchParameter(
   True=LinearInterpolate(A=LinearInterpolate(A=@n82, B=@n83, Alpha=@n84), Alpha=@n36)
   False=LinearInterpolate(A=VectorParameter() {group=background, parameterName=Background_topgradient_T, defaultValue=<rgba 0 0.0578 0.6038 1>}.RGB, Alpha=@n32)
  ) {defaultValue=true, parameterName=is known}
 ]
}
