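# Translucent UI material for a Fishing collection entry background, showing an animated underwater backdrop tinted by discovery state.
# IsDiscovered blends colors for a vertical gradient and aspect-corrected radial glow; IsKnown fades from a flat unknown color and alpha.
# Two panning, hash-jittered bubble layers tile over blended screen-space and local UVs under a radial mask.
# Screen-aligned sine waves, perturbed by scrolling noise, modulate bubble strength and overlay the top-gradient color.
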Material /Game/Athena/Collections/Art/Fishing/M_Ui_FishingCollectionENtry_Background
graph {
 properties {blendMode=BLEND_Translucent, materialDomain=MD_UI, pixelDepthOffsetMode=PDOM_Legacy, usedShadingModels=Default Lit}
 nodes [
  n1 = VectorParameter() {parameterName=RGBA Unknown, defaultValue=<rgba 0 0.0012 0.0262 0.33>}
  LinearInterpolate(A=@n1.A, Alpha=ScalarParameter() {defaultValue=1, parameterName=IsKnown, sliderMax=1})
  n2 = LinearInterpolate(
   A=VectorParameter() {parameterName=Background_topgradient_F, defaultValue=<rgba 0 0.004 0.0723 1>}.RGB
   B=VectorParameter() {group=background, parameterName=Background_topgradient_T, defaultValue=<rgba 0 0.0578 0.6038 1>}.RGB
   Alpha=ScalarParameter() {group=control, parameterName=IsDiscovered, sliderMax=1}
  )
  n3 = VectorParameter() {group=background, parameterName=Background_topgradient_T}
  n4 = VectorParameter() {group=background, parameterName=Background_glow_T, defaultValue=<rgba 0 0.6654 1 1>}
  n5 = ScalarParameter() {group=control, parameterName=IsDiscovered, sliderMax=1}
  n6 = MaterialFunctionCall(Float2=MaterialFunctionCall() {materialFunction=&GetUserInterfaceUV}.Pixel Size) {materialFunction=&BreakOutFloat2Components}
  n7 = Divide(A=@n6.R, B=Max(A=@n6.G))
  n8 = MaterialFunctionCall(Float2=TextureCoordinate()) {materialFunction=&BreakOutFloat2Components}
  n9 = MaterialFunctionCall(
   UVs=AppendVector(A=Multiply(A=@n7, B=@n8.R), B=@n8.G)
   CenterPosition=AppendVector(
    A=ScalarParameter() {defaultValue=0.5, group=background, parameterName=RadialGradient_u}
    B=ScalarParameter() {defaultValue=0.5, group=background, parameterName=RadialGradient_V}
   )
   Radius=ScalarParameter() {defaultValue=0.75, group=background, parameterName=Radial Gradient Radius_V}
   Density=ScalarParameter() {defaultValue=1, group=background, parameterName=Radial Gradient Density}
  ) {materialFunction=&RadialGradientExponential}
  n10 = VectorParameter() {parameterName=Background_base_F, defaultValue=<rgba 0.0033 0.0185 0.3763 1>}
  n11 = VectorParameter() {group=background, parameterName=Background_base_T, defaultValue=<rgba 0 0.2159 1 1>}
  n12 = ScalarParameter() {group=control, parameterName=IsDiscovered, sliderMax=1}
  n13 = TextureCoordinate()
  n14 = LinearInterpolate(
   A=@n2
   B=LinearInterpolate(
    A=LinearInterpolate(A=@n10.RGB, B=@n11.RGB, Alpha=@n12)
    B=LinearInterpolate(A=@n3.RGB, B=@n4.RGB, Alpha=@n5)
    Alpha=Saturate(None=@n9.RadialGradientExponential)
   )
   Alpha=Saturate(None=ComponentMask(None=@n13) {g=true})
  )
  n15 = LinearInterpolate(
   A=VectorParameter() {group=background, parameterName=Background_topgradient_T}.RGB
   B=VectorParameter() {group=background, parameterName=Background_glow_T}.RGB
   Alpha=ScalarParameter() {group=control, parameterName=IsDiscovered, sliderMax=1}
  )
  n16 = MaterialFunctionCall(Float2=MaterialFunctionCall() {materialFunction=&ScreenResolution}.Visible Resolution) {materialFunction=&BreakOutFloat2Components}
  n17 = Max(A=@n16.G)
  n18 = Multiply(
   A=MaterialFunctionCall() {materialFunction=&ScreenAlignedUVs}."X 100%, Y 100%"
   B=AppendVector(A=Divide(A=@n16.R, B=@n17), B=Constant() {r=1})
  )
  n19 = ScalarParameter() {defaultValue=4, group=effect, parameterName=TileScreenSpace}
  n20 = MaterialFunctionCall() {materialFunction=&GetUserInterfaceUV}
  n21 = MaterialFunctionCall(Float2=@n20.Pixel Size) {materialFunction=&BreakOutFloat2Components}
  n22 = Max(A=@n21.G)
  n23 = Multiply(A=AppendVector(A=Divide(A=@n21.R, B=@n22), B=Constant() {r=1}), B=@n20.Normalized UV)
  n24 = ScalarParameter() {defaultValue=0.25, group=effect, parameterName=Local UV}
  n25 = Multiply(
   A=Add(A=Multiply(A=@n18, B=@n19), B=Multiply(A=@n23, B=@n24))
   B=ScalarParameter() {defaultValue=0.5, group=effect, parameterName=TileAmount}
  )
  n26 = Multiply(A=Panner(Coordinate=@n25) {speedY=0.1}) {constB=4}
  n27 = MaterialFunctionCall(Float2=MaterialFunctionCall(Vec2=Floor(None=@n26)) {materialFunction=&MF_Hash}.Vec2) {materialFunction=&BreakOutFloat2Components}
  n28 = AppendVector(A=@n27.R, B=@n27.R)
  n29 = ScalarParameter() {defaultValue=0.5, group=effect, parameterName=Bubble_translation}
  n30 = MaterialFunctionCall(UVs=Add(A=Frac(None=@n26), B=Multiply(A=@n28, B=@n29) {constB=0.2}), Texture Scale=Saturate(None=@n27.R)) {materialFunction=&ScaleUVsByCenter}
  n31 = Multiply(A=Panner(Coordinate=@n25) {speedY=0.05}) {constB=3}
  n32 = MaterialFunctionCall(Float2=MaterialFunctionCall(Vec2=Floor(None=@n31)) {materialFunction=&MF_Hash}.Vec2) {materialFunction=&BreakOutFloat2Components}
  n33 = Add(
   A=Frac(None=@n31)
   B=Multiply(A=AppendVector(A=@n32.R, B=@n32.R), B=ScalarParameter() {group=effect, parameterName=Bubble_translation}) {constB=0.2}
  )
  n34 = Saturate(None=@n32.R)
  n35 = Add(
   A=Multiply(A=TextureSample(UVs=@n30.UVs) {samplerSource=SSM_Clamp_WorldGroupSettings, texture=/Game/Athena/Collections/Art/T_Ui_Bubble_64}.G) {constB=0.5}
   B=TextureSample(UVs=MaterialFunctionCall(UVs=@n33, Texture Scale=@n34) {materialFunction=&ScaleUVsByCenter}.UVs) {samplerSource=SSM_Clamp_WorldGroupSettings, texture=&T_Ui_Bubble_64}.R
  )
  n36 = Saturate(None=MaterialFunctionCall() {materialFunction=&RadialGradientExponential}.RadialGradientExponential)
  n37 = ScalarParameter() {defaultValue=1.5, group=effect, parameterName=Bubble_strenght}
  n38 = MaterialFunctionCall() {materialFunction=&ScreenAlignedUVs}
  n39 = Panner(Coordinate=@n38."X 100%, Y 100%", Speed=ScalarParameter() {defaultValue=0.025, group=Wave, parameterName=WaveSpeed_vert}) {speedX=0.1, speedY=0.015}
  n40 = ScalarParameter() {defaultValue=0.25, group=Wave, parameterName=WaveSIze}
  n41 = TextureSample(UVs=Panner(Coordinate=@n38."X 100%, Y 100%", Speed=ScalarParameter() {defaultValue=0.015, group=Wave, parameterName=WaveSpeed_hor}) {speedX=0.1}) {samplerType=SAMPLERTYPE_Grayscale, texture=&LowResBlurredNoise}
  n42 = Multiply(A=@n41.R, B=@n41.R) {constB=10}
  n43 = Add(
   A=Sine(None=Multiply(A=@n39, B=@n40))
   B=LinearInterpolate(B=ScalarParameter() {defaultValue=10, group=Wave, parameterName=Wave_transform}, Alpha=Saturate(None=@n42)) {constB=10}
  )
  n44 = Saturate(None=Subtract(A=ComponentMask(None=@n43) {g=true}) {constB=0.1})
  n45 = Saturate(None=Multiply(A=Multiply(A=@n35, B=@n36), B=LinearInterpolate(A=@n37, Alpha=@n44) {constA=1}))
  n46 = ScalarParameter() {defaultValue=0.5, group=Wave, parameterName=WaveOpacity, sliderMax=1}
  LinearInterpolate(
   A=@n1.RGB
   B=LinearInterpolate(A=LinearInterpolate(A=@n14, B=@n15, Alpha=@n45), B=@n2, Alpha=Multiply(A=@n44, B=@n46))
   Alpha=ScalarParameter() {defaultValue=1, parameterName=IsKnown, sliderMax=1}
  )
 ]
}
